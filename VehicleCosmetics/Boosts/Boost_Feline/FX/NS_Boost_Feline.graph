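# Niagara system for the Feline vehicle boost cosmetic: a row of seven red mesh flares that light in a back-and-forth sweep.
# Small and Tall flare layers are each split across two emitters, burst-spawning by spawn group at local Y offsets from System.FinalPosition.
# Per-group staggered curves pulse color on both layers and stretch the small flares.
# Spawn group also feeds renderer visibility tags; velocity and paint-override modules are disabled.
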NiagaraSystem /VehicleCosmetics/Boosts/Boost_Feline/FX/NS_Boost_Feline
dependencies {
 usedDataInterfaces [&NiagaraDataInterfaceVectorCurve, &NiagaraDataInterfaceSkeletalMesh]
 usedDynamicInputs []
 usedModules [
  &self:SystemSpawnScript.SystemScriptSource.SystemScriptGraph.NiagaraNodeAssignment_0.SetVariables_#1
  &SystemState
 ]
}
emitters {
 NE_VC_SingleBoost_SmallFlares {
  emitterUpdateScript [
   &EmitterState()
   &SpawnBurst_Instantaneous(Spawn Group=4, Use Spawn Probability=false, UseLoopCountLimit=false) {module=SpawnBurst_Instantaneous004}
   &SpawnBurst_Instantaneous(Use Spawn Probability=false, UseLoopCountLimit=false)
   &SpawnBurst_Instantaneous(Spawn Group=1, Use Spawn Probability=false, UseLoopCountLimit=false) {module=SpawnBurst_Instantaneous001}
   &SpawnBurst_Instantaneous(Spawn Group=2, Use Spawn Probability=false, UseLoopCountLimit=false) {module=SpawnBurst_Instantaneous002}
   &SpawnBurst_Instantaneous(Spawn Group=3, Use Spawn Probability=false, UseLoopCountLimit=false) {module=SpawnBurst_Instantaneous003}
   &SpawnRate(Use Spawn Probability=false) {enabled=false}
  ]
  particleSpawnScript [
   &VC_PaintedVariantsOverride() {enabled=false}
   &InitializeParticle(
    Color=<rgba 1 0.05 0.075 1>
    Lifetime=1.145
    Mesh Scale=&Multiply_VectorByFloat(Vector=<12.5 12.5 12.5>, Float=4)
    Mesh Scale Mode=::NewEnumerator4
    Position Offset=/VehicleCosmetics/FX/Niagara/Modules/MergeEmitters/VC_VectorBySpawnGroup(
     SpawnGroup or ExecIndex:NiagaraInt32=0
     CustomInt:NiagaraInt32=0
     MaximumSpawnGroups:NiagaraInt32=5
     Vector Group 0:Vector3f=&AddVectorToPosition(Position=<0 40 0>, Vector=$System.FinalPosition)
     Vector Group 1:Vector3f=&AddVectorToPosition(Position=<0 25 0>, Vector=$System.FinalPosition)
     Vector Group 2:Vector3f=&AddVectorToPosition(Position=<0 12.5 0>, Vector=$System.FinalPosition)
     Vector Group 3:Vector3f=&AddVectorToPosition(Position=<0 0 0>, Vector=$System.FinalPosition)
     Vector Group 4:Vector3f=&AddVectorToPosition(Position=<0 -12.5 0>, Vector=$System.FinalPosition)
     Vector Group 5:Vector3f=<0 0 0>
    )
    Position Offset Coordinate Space=::Local
    Sprite Rotation Mode=::NewEnumerator2
    UsePositionOffset=true
    Write Material Random=true
   )
   &InitialMeshOrientation(Mesh Orientation Mode=::NewEnumerator2, Orientation Vector=<-1 0 0>, Use Rotation Vector=false)
   &AddVelocity(Velocity=<-1500 0 0>) {enabled=false}
  ]
  particleUpdateScript [
   /SetVariables(Particles.FelTagVis:NiagaraInt32=&SpawnGroup()) {isSetParameters=true, module=SetVariables_#2}
   &ParticleState()
   &ScaleMeshSize(Scale Factor=&VC_VectorBySpawnGroup(
     SpawnGroup or ExecIndex:NiagaraInt32=0
     CustomInt:NiagaraInt32=0
     MaximumSpawnGroups:NiagaraInt32=5
     Vector Group 0:Vector3f=&VectorFromCurve(
      VectorCurve=DataInterface() {propertyValues {
        XCurve {keys [<at 0 1>,<at 0.12 15>,<at 0.19 1>,<at 0.87 1>,<at 0.88 15>,<at 1 1>]}
        YCurve {keys [<at 0 1>,<at 0.12 1>,<at 0.19 1>,<at 0.87 1>,<at 0.88 1>,<at 1 1>]}
        ZCurve {keys [<at 0 1>,<at 0.12 1>,<at 0.19 1>,<at 0.87 1>,<at 0.88 1>,<at 1 1>]}
       }}
      Scale Curve=<0.1 1 1>
     )
     Vector Group 1:Vector3f=&VectorFromCurve(
      VectorCurve=DataInterface() {propertyValues {
        XCurve {keys [<at 0.12 1>,<at 0.19 15>,<at 0.25 1>,<at 0.805 1>,<at 0.81 15>,<at 0.88 1>]}
        YCurve {keys [<at 0.12 1>,<at 0.19 1>,<at 0.25 1>,<at 0.805 1>,<at 0.81 1>,<at 0.88 1>]}
        ZCurve {keys [<at 0.12 1>,<at 0.19 1>,<at 0.25 1>,<at 0.805 1>,<at 0.81 1>,<at 0.88 1>]}
       }}
      Scale Curve=<0.1 1 1>
     )
     Vector Group 2:Vector3f=&VectorFromCurve(
      VectorCurve=DataInterface() {propertyValues {
        XCurve {keys [<at 0.19 1>,<at 0.25 15>,<at 0.31 1>,<at 0.74 1>,<at 0.75 15>,<at 0.81 1>]}
        YCurve {keys [<at 0.19 1>,<at 0.19 1>,<at 0.25 1>,<at 0.31 1>,<at 0.74 1>,<at 0.75 1>,<at 0.81 1>]}
        ZCurve {keys [<at 0.19 1>,<at 0.19 1>,<at 0.25 1>,<at 0.31 1>,<at 0.74 1>,<at 0.75 1>,<at 0.81 1>]}
       }}
      Scale Curve=<0.1 1 1>
     )
     Vector Group 3:Vector3f=&VectorFromCurve(
      VectorCurve=DataInterface() {propertyValues {
        XCurve {keys [<at 0.24 1>,<at 0.31 15>,<at 0.38 1>,<at 0.69 1>,<at 0.7 15>,<at 0.75 1>]}
        YCurve {keys [<at 0.24 1>,<at 0.24 1>,<at 0.31 1>,<at 0.38 1>,<at 0.69 1>,<at 0.7 1>,<at 0.75 1>]}
        ZCurve {keys [<at 0.24 1>,<at 0.24 1>,<at 0.31 1>,<at 0.38 1>,<at 0.69 1>,<at 0.7 1>,<at 0.75 1>]}
       }}
      Scale Curve=<0.1 1 1>
     )
     Vector Group 4:Vector3f=&VectorFromCurve(
      VectorCurve=DataInterface() {propertyValues {
        XCurve {keys [<at 0.31 1>,<at 0.38 15>,<at 0.44 1>,<at 0.62 1>,<at 0.63 15>,<at 0.69 1>]}
        YCurve {keys [<at 0.31 1>,<at 0.31 1>,<at 0.38 1>,<at 0.44 1>,<at 0.62 1>,<at 0.63 1>,<at 0.69 1>]}
        ZCurve {keys [<at 0.31 1>,<at 0.31 1>,<at 0.38 1>,<at 0.44 1>,<at 0.62 1>,<at 0.63 1>,<at 0.69 1>]}
       }}
      Scale Curve=<0.1 1 1>
     )
     Vector Group 5:Vector3f=<0 0 0>
    ))
   &ScaleColor(
    Scale RGB=&VC_VectorBySpawnGroup(
     SpawnGroup or ExecIndex:NiagaraInt32=0
     CustomInt:NiagaraInt32=0
     MaximumSpawnGroups:NiagaraInt32=5
     Vector Group 0:Vector3f=&VectorFromCurve(
      VectorCurve=DataInterface() {propertyValues {
        XCurve {
         defaultValue=0
         keys [
          {tangentMode=RCTM_User, time=0.062, value=1}
          {arriveTangentWeight=0.02, interpMode=RCIM_Cubic, time=0.12, value=200}
          <at 0.19 1>
          {tangentMode=RCTM_Break, time=0.87, value=1}
          <at 0.88 200 RCIM_Cubic>
          {tangentMode=RCTM_User, time=0.93, value=1}
         ]
        }
        YCurve {
         defaultValue=0
         keys [
          {tangentMode=RCTM_User, time=0.062, value=1}
          {arriveTangentWeight=0.02, interpMode=RCIM_Cubic, time=0.12, value=200}
          <at 0.19 1>
          {tangentMode=RCTM_Break, time=0.87, value=1}
          <at 0.88 200 RCIM_Cubic>
          {tangentMode=RCTM_User, time=0.93, value=1}
         ]
        }
        ZCurve {keys [
          {tangentMode=RCTM_User, time=0.062, value=1}
          {arriveTangentWeight=0.02, interpMode=RCIM_Cubic, time=0.12, value=200}
          <at 0.19 1>
          {tangentMode=RCTM_Break, time=0.87, value=1}
          <at 0.88 200 RCIM_Cubic>
          {tangentMode=RCTM_User, time=0.93, value=1}
         ]}
       }}
      CurveIndex=$Emitter.NormalizedLoopAge
     )
     Vector Group 1:Vector3f=&VectorFromCurve(
      VectorCurve=DataInterface() {propertyValues {
        XCurve {
         defaultValue=0
         keys [
          {tangentMode=RCTM_User, time=0.12, value=1}
          <at 0.19 500 RCIM_Cubic>
          {arriveTangentWeight=0.02, tangentMode=RCTM_User, time=0.25, value=1}
          {tangentMode=RCTM_Break, time=0.805, value=1}
          <at 0.81 500 RCIM_Cubic>
          {tangentMode=RCTM_User, time=0.88, value=1}
         ]
        }
        YCurve {
         defaultValue=0
         keys [
          {tangentMode=RCTM_User, time=0.12, value=1}
          <at 0.19 500 RCIM_Cubic>
          {arriveTangentWeight=0.02, tangentMode=RCTM_User, time=0.25, value=1}
          {tangentMode=RCTM_Break, time=0.805, value=1}
          <at 0.81 500 RCIM_Cubic>
          {tangentMode=RCTM_User, time=0.88, value=1}
         ]
        }
        ZCurve {keys [
          {tangentMode=RCTM_User, time=0.12, value=1}
          <at 0.19 500 RCIM_Cubic>
          {arriveTangentWeight=0.02, tangentMode=RCTM_User, time=0.25, value=1}
          {tangentMode=RCTM_Break, time=0.805, value=1}
          <at 0.81 500 RCIM_Cubic>
          {tangentMode=RCTM_User, time=0.88, value=1}
         ]}
       }}
      CurveIndex=$Emitter.NormalizedLoopAge
      Scale Curve=<0.5 0.5 0.5>
     )
     Vector Group 2:Vector3f=&VectorFromCurve(
      VectorCurve=DataInterface() {propertyValues {
        XCurve {
         defaultValue=0
         keys [
          {tangentMode=RCTM_User, time=0.19, value=1}
          {arriveTangentWeight=0.02, interpMode=RCIM_Cubic, time=0.25, value=500}
          <at 0.31 1>
          {tangentMode=RCTM_Break, time=0.74, value=1}
          <at 0.75 500 RCIM_Cubic>
          {tangentMode=RCTM_User, time=0.81, value=1}
         ]
        }
        YCurve {
         defaultValue=0
         keys [
          {tangentMode=RCTM_User, time=0.19, value=1}
          {arriveTangentWeight=0.02, interpMode=RCIM_Cubic, time=0.25, value=500}
          <at 0.31 1>
          {tangentMode=RCTM_Break, time=0.74, value=1}
          <at 0.75 500 RCIM_Cubic>
          {tangentMode=RCTM_User, time=0.81, value=1}
         ]
        }
        ZCurve {keys [
          {tangentMode=RCTM_User, time=0.19, value=1}
          {arriveTangentWeight=0.02, interpMode=RCIM_Cubic, time=0.25, value=500}
          <at 0.31 1>
          {tangentMode=RCTM_Break, time=0.74, value=1}
          <at 0.75 500 RCIM_Cubic>
          {tangentMode=RCTM_User, time=0.81, value=1}
         ]}
       }}
      CurveIndex=$Emitter.NormalizedLoopAge
      Scale Curve=<0.5 0.5 0.5>
     )
     Vector Group 3:Vector3f=&VectorFromCurve(
      VectorCurve=DataInterface() {propertyValues {
        XCurve {
         defaultValue=0
         keys [
          {tangentMode=RCTM_User, time=0.25, value=1}
          {arriveTangentWeight=0.02, interpMode=RCIM_Cubic, time=0.31, value=200}
          <at 0.38 1>
          {tangentMode=RCTM_Break, time=0.69, value=1}
          <at 0.7 200 RCIM_Cubic>
          {tangentMode=RCTM_User, time=0.75, value=1}
         ]
        }
        YCurve {
         defaultValue=0
         keys [
          {tangentMode=RCTM_User, time=0.25, value=1}
          {arriveTangentWeight=0.02, interpMode=RCIM_Cubic, time=0.31, value=200}
          <at 0.38 1>
          {tangentMode=RCTM_Break, time=0.69, value=1}
          <at 0.7 200 RCIM_Cubic>
          {tangentMode=RCTM_User, time=0.75, value=1}
         ]
        }
        ZCurve {keys [
          {tangentMode=RCTM_User, time=0.25, value=1}
          {arriveTangentWeight=0.02, interpMode=RCIM_Cubic, time=0.31, value=200}
          <at 0.38 1>
          {tangentMode=RCTM_Break, time=0.69, value=1}
          <at 0.7 200 RCIM_Cubic>
          {tangentMode=RCTM_User, time=0.75, value=1}
         ]}
       }}
      CurveIndex=$Emitter.NormalizedLoopAge
      Scale Curve=<0.5 0.5 0.5>
     )
     Vector Group 4:Vector3f=&VectorFromCurve(
      VectorCurve=DataInterface() {propertyValues {
        XCurve {
         defaultValue=0
         keys [
          {tangentMode=RCTM_User, time=0.31, value=1}
          {arriveTangentWeight=0.02, interpMode=RCIM_Cubic, time=0.38, value=500}
          <at 0.44 1>
          {tangentMode=RCTM_Break, time=0.62, value=1}
          <at 0.63 500 RCIM_Cubic>
          {tangentMode=RCTM_User, time=0.69, value=1}
         ]
        }
        YCurve {
         defaultValue=0
         keys [
          {tangentMode=RCTM_User, time=0.31, value=1}
          {arriveTangentWeight=0.02, interpMode=RCIM_Cubic, time=0.38, value=500}
          <at 0.44 1>
          {tangentMode=RCTM_Break, time=0.62, value=1}
          <at 0.63 500 RCIM_Cubic>
          {tangentMode=RCTM_User, time=0.69, value=1}
         ]
        }
        ZCurve {keys [
          {tangentMode=RCTM_User, time=0.31, value=1}
          {arriveTangentWeight=0.02, interpMode=RCIM_Cubic, time=0.38, value=500}
          <at 0.44 1>
          {tangentMode=RCTM_Break, time=0.62, value=1}
          <at 0.63 500 RCIM_Cubic>
          {tangentMode=RCTM_User, time=0.69, value=1}
         ]}
       }}
      CurveIndex=$Emitter.NormalizedLoopAge
      Scale Curve=<0.5 0.5 0.5>
     )
     Vector Group 5:Vector3f=<0 0 0>
    )
    ScaleA=true
    ScaleRGB=true
   )
   &UpdateMeshOrientation()
   &SolveForcesAndVelocity(Clamp Velocity=false, Limit Acceleration=false)
  ]
  rendererClasses [&NiagaraSpriteRendererProperties, &NiagaraMeshRendererProperties]
  renderers [
   {rendererClass=&NiagaraSpriteRendererProperties}
   {rendererClass=&NiagaraMeshRendererProperties, rendererIndex=1}
   {rendererClass=&NiagaraMeshRendererProperties, rendererIndex=2}
   {rendererClass=&NiagaraMeshRendererProperties, rendererIndex=3}
   {rendererClass=&NiagaraMeshRendererProperties, rendererIndex=4}
   {rendererClass=&NiagaraMeshRendererProperties, rendererIndex=5}
  ]
 }
 NE_VC_SingleBoost_SmallFlares001 {
  emitterUpdateScript [
   &EmitterState()
   &SpawnBurst_Instantaneous(Use Spawn Probability=false, UseLoopCountLimit=false)
   &SpawnBurst_Instantaneous(Spawn Group=1, Use Spawn Probability=false, UseLoopCountLimit=false) {module=SpawnBurst_Instantaneous001}
   &SpawnRate(Use Spawn Probability=false) {enabled=false}
  ]
  particleSpawnScript [
   &VC_PaintedVariantsOverride() {enabled=false}
   &InitializeParticle(
    Color=<rgba 1 0.05 0.075 1>
    Lifetime=1.145
    Mesh Scale=&Multiply_VectorByFloat(Vector=<12.5 12.5 12.5>, Float=4)
    Mesh Scale Mode=::NewEnumerator4
    Position Offset=&VC_VectorBySpawnGroup(
     SpawnGroup or ExecIndex:NiagaraInt32=0
     CustomInt:NiagaraInt32=0
     MaximumSpawnGroups:NiagaraInt32=2
     Vector Group 0:Vector3f=&AddVectorToPosition(Position=<0 -25 0>, Vector=$System.FinalPosition)
     Vector Group 1:Vector3f=&AddVectorToPosition(Position=<0 -40 0>, Vector=$System.FinalPosition)
     Vector Group 2:Vector3f=<-18 4 0>
     Vector Group 3:Vector3f=<-18 0 0>
     Vector Group 4:Vector3f=<-18 -4 0>
     Vector Group 5:Vector3f=<0 0 0>
    )
    Position Offset Coordinate Space=::Local
    Sprite Rotation Mode=::NewEnumerator2
    UsePositionOffset=true
    Write Material Random=true
   )
   &InitialMeshOrientation(Mesh Orientation Mode=::NewEnumerator2, Orientation Vector=<-1 0 0>, Use Rotation Vector=false)
   &AddVelocity(Velocity=<-1500 0 0>) {enabled=false}
  ]
  particleUpdateScript [
   /SetVariables(Particles.KTagVis:NiagaraInt32=&SpawnGroup()) {isSetParameters=true, module=SetVariables_#2}
   &ParticleState()
   &ScaleMeshSize(Scale Factor=&VC_VectorBySpawnGroup(
     SpawnGroup or ExecIndex:NiagaraInt32=0
     CustomInt:NiagaraInt32=0
     MaximumSpawnGroups:NiagaraInt32=2
     Vector Group 0:Vector3f=&VectorFromCurve(
      VectorCurve=DataInterface() {propertyValues {
        XCurve {keys [<at 0.38 1>,<at 0.44 15>,<at 0.5 1>,<at 0.55 1>,<at 0.56 15>,<at 0.62 1>]}
        YCurve {keys [<at 0.38 1>,<at 0.38 1>,<at 0.44 1>,<at 0.5 1>,<at 0.55 1>,<at 0.56 1>,<at 0.62 1>]}
        ZCurve {keys [<at 0.38 1>,<at 0.38 1>,<at 0.44 1>,<at 0.5 1>,<at 0.55 1>,<at 0.56 1>,<at 0.62 1>]}
       }}
      Scale Curve=<0.1 1 1>
     )
     Vector Group 1:Vector3f=&VectorFromCurve(
      VectorCurve=DataInterface() {propertyValues {
        XCurve {keys [<at 0.44 1 RCIM_Cubic>,<at 0.5 15 RCIM_Cubic>,<at 0.56 1>]}
        YCurve {keys [<at 0.44 1 RCIM_Cubic>,<at 0.44 1 RCIM_Cubic>,<at 0.5 1>,<at 0.56 1>]}
        ZCurve {keys [<at 0.44 1 RCIM_Cubic>,<at 0.44 1 RCIM_Cubic>,<at 0.5 1>,<at 0.56 1 RCIM_Cubic>]}
       }}
      Scale Curve=<0.1 1 1>
     )
     Vector Group 2:Vector3f=<0 0 0>
     Vector Group 3:Vector3f=<0 0 0>
     Vector Group 4:Vector3f=<0 0 0>
     Vector Group 5:Vector3f=<0 0 0>
    ))
   &ScaleColor(
    Scale RGB=&VC_VectorBySpawnGroup(
     SpawnGroup or ExecIndex:NiagaraInt32=0
     CustomInt:NiagaraInt32=0
     MaximumSpawnGroups:NiagaraInt32=2
     Vector Group 0:Vector3f=&VectorFromCurve(
      VectorCurve=DataInterface() {propertyValues {
        XCurve {
         defaultValue=0
         keys [
          <at 0.38 1 RCIM_Cubic>
          {arriveTangentWeight=0.02, interpMode=RCIM_Cubic, time=0.44, value=500}
          {arriveTangent=-4158.333, leaveTangent=-4158.333, time=0.5, value=1}
          {arriveTangent=7128.5723, interpMode=RCIM_Cubic, leaveTangent=7128.5723, time=0.56, value=1}
          <at 0.57 500 RCIM_Cubic>
          <at 0.62 1>
         ]
        }
        YCurve {
         defaultValue=0
         keys [
          <at 0.38 1 RCIM_Cubic>
          {arriveTangentWeight=0.02, interpMode=RCIM_Cubic, time=0.44, value=500}
          {arriveTangent=-4158.333, leaveTangent=-4158.333, time=0.5, value=1}
          {arriveTangent=7128.5723, interpMode=RCIM_Cubic, leaveTangent=7128.5723, time=0.56, value=1}
          <at 0.57 500 RCIM_Cubic>
          <at 0.62 1>
         ]
        }
        ZCurve {keys [
          <at 0.38 1 RCIM_Cubic>
          {arriveTangentWeight=0.02, interpMode=RCIM_Cubic, time=0.44, value=500}
          {arriveTangent=-4158.333, leaveTangent=-4158.333, time=0.5, value=1}
          {arriveTangent=7128.5723, interpMode=RCIM_Cubic, leaveTangent=7128.5723, time=0.56, value=1}
          <at 0.57 500 RCIM_Cubic>
          <at 0.62 1>
         ]}
       }}
      CurveIndex=$Emitter.NormalizedLoopAge
      Scale Curve=<0.5 0.5 0.5>
     )
     Vector Group 1:Vector3f=&VectorFromCurve(
      VectorCurve=DataInterface() {propertyValues {
        XCurve {
         defaultValue=0
         keys [<at 0.44 1 RCIM_Cubic>, {arriveTangentWeight=0.02, interpMode=RCIM_Cubic, time=0.5, value=500}, {tangentMode=RCTM_User, time=0.56, value=1}]
        }
        YCurve {
         defaultValue=0
         keys [<at 0.44 1 RCIM_Cubic>, {arriveTangentWeight=0.02, interpMode=RCIM_Cubic, time=0.5, value=500}, {tangentMode=RCTM_User, time=0.56, value=1}]
        }
        ZCurve {keys [<at 0.44 1 RCIM_Cubic>, {arriveTangentWeight=0.02, interpMode=RCIM_Cubic, time=0.5, value=500}, {tangentMode=RCTM_User, time=0.56, value=1}]}
       }}
      CurveIndex=$Emitter.NormalizedLoopAge
      Scale Curve=<0.5 0.5 0.5>
     )
     Vector Group 2:Vector3f=<0 0 0>
     Vector Group 3:Vector3f=<0 0 0>
     Vector Group 4:Vector3f=<0 0 0>
     Vector Group 5:Vector3f=<0 0 0>
    )
    ScaleA=true
    ScaleRGB=true
   )
   &UpdateMeshOrientation()
   &SolveForcesAndVelocity(Clamp Velocity=false, Limit Acceleration=false)
  ]
  rendererClasses [&NiagaraSpriteRendererProperties, &NiagaraMeshRendererProperties]
  renderers [
   {rendererClass=&NiagaraSpriteRendererProperties}
   {rendererClass=&NiagaraMeshRendererProperties, rendererIndex=1}
   {rendererClass=&NiagaraMeshRendererProperties, rendererIndex=2}
  ]
 }
 NE_VC_SingleBoost_TallFlares {
  emitterUpdateScript [
   &EmitterState()
   &SpawnBurst_Instantaneous(Use Spawn Probability=false, UseLoopCountLimit=false)
   &SpawnBurst_Instantaneous(Spawn Group=1, Use Spawn Probability=false, UseLoopCountLimit=false) {module=SpawnBurst_Instantaneous001}
   &SpawnBurst_Instantaneous(Spawn Group=2, Use Spawn Probability=false, UseLoopCountLimit=false) {module=SpawnBurst_Instantaneous002}
   &SpawnBurst_Instantaneous(Spawn Group=3, Use Spawn Probability=false, UseLoopCountLimit=false) {module=SpawnBurst_Instantaneous003}
   &SpawnBurst_Instantaneous(Spawn Group=4, Use Spawn Probability=false, UseLoopCountLimit=false) {module=SpawnBurst_Instantaneous004}
   &SpawnRate(Use Spawn Probability=false) {enabled=false}
  ]
  particleSpawnScript [
   &VC_PaintedVariantsOverride() {enabled=false}
   &InitializeParticle(
    Color=<rgba 1 0.05 0.075 1>
    Lifetime=1.145
    Mesh Scale=&Multiply_VectorByFloat(Vector=<2.75 12.5 12.5>, Float=4)
    Mesh Scale Mode=::NewEnumerator4
    Position Offset=&VC_VectorBySpawnGroup(
     SpawnGroup or ExecIndex:NiagaraInt32=0
     CustomInt:NiagaraInt32=0
     MaximumSpawnGroups:NiagaraInt32=5
     Vector Group 0:Vector3f=&AddVectorToPosition(Position=<0 40 0>, Vector=$System.FinalPosition)
     Vector Group 1:Vector3f=&AddVectorToPosition(Position=<0 25 0>, Vector=$System.FinalPosition)
     Vector Group 2:Vector3f=&AddVectorToPosition(Position=<0 12.5 0>, Vector=$System.FinalPosition)
     Vector Group 3:Vector3f=&AddVectorToPosition(Position=<0 0 0>, Vector=$System.FinalPosition)
     Vector Group 4:Vector3f=&AddVectorToPosition(Position=<0 -12.5 0>, Vector=$System.FinalPosition)
     Vector Group 5:Vector3f=<0 0 0>
    )
    Position Offset Coordinate Space=::Local
    Sprite Rotation Mode=::NewEnumerator2
    UsePositionOffset=true
    Write Material Random=true
   )
   &AddVelocity(Velocity=<-1500 0 0>) {enabled=false}
   &InitialMeshOrientation(Mesh Orientation Mode=::NewEnumerator2, Orientation Vector=<-1 0 0>, Use Rotation Vector=false)
  ]
  particleUpdateScript [
   /SetVariables(Particles.FeTagVis:NiagaraInt32=&SpawnGroup()) {isSetParameters=true, module=SetVariables_#3}
   &ParticleState()
   &ScaleColor(
    Scale RGB=&VC_VectorBySpawnGroup(
     SpawnGroup or ExecIndex:NiagaraInt32=0
     CustomInt:NiagaraInt32=0
     MaximumSpawnGroups:NiagaraInt32=5
     Vector Group 0:Vector3f=&VectorFromCurve(VectorCurve=DataInterface() {propertyValues {
        XCurve {
         defaultValue=0
         keys [
          {tangentMode=RCTM_User, time=0.06, value=1}
          {arriveTangentWeight=0.02, tangentMode=RCTM_User, time=0.12, value=150}
          <at 0.19 100>
          {tangentMode=RCTM_Break, time=0.87, value=1}
          <at 0.88 10>
          {tangentMode=RCTM_User, time=1, value=1}
         ]
        }
        YCurve {
         defaultValue=0
         keys [
          {tangentMode=RCTM_User, time=0.06, value=1}
          {arriveTangentWeight=0.02, tangentMode=RCTM_User, time=0.12, value=150}
          <at 0.19 100>
          {tangentMode=RCTM_Break, time=0.87, value=1}
          <at 0.88 10>
          {tangentMode=RCTM_User, time=1, value=1}
         ]
        }
        ZCurve {keys [
          {tangentMode=RCTM_User, time=0.06, value=1}
          {arriveTangentWeight=0.02, tangentMode=RCTM_User, time=0.12, value=150}
          <at 0.19 100>
          {tangentMode=RCTM_Break, time=0.87, value=1}
          <at 0.88 10>
          {tangentMode=RCTM_User, time=1, value=1}
         ]}
       }})
     Vector Group 1:Vector3f=&VectorFromCurve(VectorCurve=DataInterface() {propertyValues {
        XCurve {
         defaultValue=0
         keys [
          {tangentMode=RCTM_User, time=0.12, value=1}
          <at 0.19 150>
          {arriveTangentWeight=0.02, tangentMode=RCTM_User, time=0.25, value=100}
          {tangentMode=RCTM_Break, time=0.805, value=1}
          <at 0.81 10>
          {tangentMode=RCTM_User, time=1, value=1}
         ]
        }
        YCurve {
         defaultValue=0
         keys [
          {tangentMode=RCTM_User, time=0.12, value=1}
          <at 0.19 150>
          {arriveTangentWeight=0.02, tangentMode=RCTM_User, time=0.25, value=100}
          {tangentMode=RCTM_Break, time=0.805, value=1}
          <at 0.81 10>
          {tangentMode=RCTM_User, time=1, value=1}
         ]
        }
        ZCurve {keys [
          {tangentMode=RCTM_User, time=0.12, value=1}
          <at 0.19 150>
          {arriveTangentWeight=0.02, tangentMode=RCTM_User, time=0.25, value=100}
          {tangentMode=RCTM_Break, time=0.805, value=1}
          <at 0.81 10>
          {tangentMode=RCTM_User, time=1, value=1}
         ]}
       }})
     Vector Group 2:Vector3f=&VectorFromCurve(VectorCurve=DataInterface() {propertyValues {
        XCurve {
         defaultValue=0
         keys [
          {tangentMode=RCTM_User, time=0.19, value=1}
          {arriveTangentWeight=0.02, tangentMode=RCTM_User, time=0.25, value=150}
          <at 0.31 100>
          {tangentMode=RCTM_Break, time=0.74, value=1}
          <at 0.75 10>
          {tangentMode=RCTM_User, time=1, value=1}
         ]
        }
        YCurve {
         defaultValue=0
         keys [
          {tangentMode=RCTM_User, time=0.19, value=1}
          {arriveTangentWeight=0.02, tangentMode=RCTM_User, time=0.25, value=150}
          <at 0.31 100>
          {tangentMode=RCTM_Break, time=0.74, value=1}
          <at 0.75 10>
          {tangentMode=RCTM_User, time=1, value=1}
         ]
        }
        ZCurve {keys [
          {tangentMode=RCTM_User, time=0.19, value=1}
          {arriveTangentWeight=0.02, tangentMode=RCTM_User, time=0.25, value=150}
          <at 0.31 100>
          {tangentMode=RCTM_Break, time=0.74, value=1}
          <at 0.75 10>
          {tangentMode=RCTM_User, time=1, value=1}
         ]}
       }})
     Vector Group 3:Vector3f=&VectorFromCurve(VectorCurve=DataInterface() {propertyValues {
        XCurve {
         defaultValue=0
         keys [
          {tangentMode=RCTM_User, time=0.25, value=1}
          {arriveTangentWeight=0.02, tangentMode=RCTM_User, time=0.31, value=150}
          <at 0.38 100>
          {tangentMode=RCTM_Break, time=0.69, value=1}
          <at 0.7 10>
          {tangentMode=RCTM_User, time=1, value=1}
         ]
        }
        YCurve {
         defaultValue=0
         keys [
          {tangentMode=RCTM_User, time=0.25, value=1}
          {arriveTangentWeight=0.02, tangentMode=RCTM_User, time=0.31, value=150}
          <at 0.38 100>
          {tangentMode=RCTM_Break, time=0.69, value=1}
          <at 0.7 10>
          {tangentMode=RCTM_User, time=1, value=1}
         ]
        }
        ZCurve {keys [
          {tangentMode=RCTM_User, time=0.25, value=1}
          {arriveTangentWeight=0.02, tangentMode=RCTM_User, time=0.31, value=150}
          <at 0.38 100>
          {tangentMode=RCTM_Break, time=0.69, value=1}
          <at 0.7 10>
          {tangentMode=RCTM_User, time=1, value=1}
         ]}
       }})
     Vector Group 4:Vector3f=&VectorFromCurve(VectorCurve=DataInterface() {propertyValues {
        XCurve {
         defaultValue=0
         keys [
          {tangentMode=RCTM_User, time=0.31, value=1}
          {arriveTangentWeight=0.02, tangentMode=RCTM_User, time=0.38, value=150}
          <at 0.44 100>
          {tangentMode=RCTM_Break, time=0.62, value=1}
          <at 0.63 10>
          {tangentMode=RCTM_User, time=1, value=1}
         ]
        }
        YCurve {
         defaultValue=0
         keys [
          {tangentMode=RCTM_User, time=0.31, value=1}
          {arriveTangentWeight=0.02, tangentMode=RCTM_User, time=0.38, value=150}
          <at 0.44 100>
          {tangentMode=RCTM_Break, time=0.62, value=1}
          <at 0.63 10>
          {tangentMode=RCTM_User, time=1, value=1}
         ]
        }
        ZCurve {keys [
          {tangentMode=RCTM_User, time=0.31, value=1}
          {arriveTangentWeight=0.02, tangentMode=RCTM_User, time=0.38, value=150}
          <at 0.44 100>
          {tangentMode=RCTM_Break, time=0.62, value=1}
          <at 0.63 10>
          {tangentMode=RCTM_User, time=1, value=1}
         ]}
       }})
     Vector Group 5:Vector3f=<0 0 0>
    )
    ScaleA=true
    ScaleRGB=true
   )
   &UpdateMeshOrientation()
   &SolveForcesAndVelocity(Clamp Velocity=false, Limit Acceleration=false)
  ]
  rendererClasses [&NiagaraSpriteRendererProperties, &NiagaraMeshRendererProperties]
  renderers [
   {rendererClass=&NiagaraSpriteRendererProperties}
   {rendererClass=&NiagaraMeshRendererProperties, rendererIndex=1}
   {rendererClass=&NiagaraMeshRendererProperties, rendererIndex=2}
   {rendererClass=&NiagaraMeshRendererProperties, rendererIndex=3}
   {rendererClass=&NiagaraMeshRendererProperties, rendererIndex=4}
   {rendererClass=&NiagaraMeshRendererProperties, rendererIndex=5}
  ]
 }
 NE_VC_SingleBoost_TallFlares001 {
  emitterUpdateScript [
   &EmitterState()
   &SpawnBurst_Instantaneous(Use Spawn Probability=false, UseLoopCountLimit=false)
   &SpawnBurst_Instantaneous(Spawn Group=1, Use Spawn Probability=false, UseLoopCountLimit=false) {module=SpawnBurst_Instantaneous001}
   &SpawnRate(Use Spawn Probability=false) {enabled=false}
  ]
  particleSpawnScript [
   &VC_PaintedVariantsOverride() {enabled=false}
   &InitializeParticle(
    Color=<rgba 1 0.05 0.075 1>
    Lifetime=1.145
    Mesh Scale=&Multiply_VectorByFloat(Vector=<2.75 12.5 12.5>, Float=4)
    Mesh Scale Mode=::NewEnumerator4
    Position Offset=&VC_VectorBySpawnGroup(
     SpawnGroup or ExecIndex:NiagaraInt32=0
     CustomInt:NiagaraInt32=0
     MaximumSpawnGroups:NiagaraInt32=2
     Vector Group 0:Vector3f=&AddVectorToPosition(Position=<0 -25 0>, Vector=$System.FinalPosition)
     Vector Group 1:Vector3f=&AddVectorToPosition(Position=<0 -40 0>, Vector=$System.FinalPosition)
     Vector Group 2:Vector3f=<0 80 0>
     Vector Group 3:Vector3f=<0 0 0>
     Vector Group 4:Vector3f=<0 -80 0>
     Vector Group 5:Vector3f=<0 0 0>
    )
    Position Offset Coordinate Space=::Local
    Sprite Rotation Mode=::NewEnumerator2
    UsePositionOffset=true
    Write Material Random=true
   )
   &AddVelocity(Velocity=<-1500 0 0>) {enabled=false}
   &InitialMeshOrientation(Mesh Orientation Mode=::NewEnumerator2, Orientation Vector=<-1 0 0>, Use Rotation Vector=false)
  ]
  particleUpdateScript [
   /SetVariables(Particles.FeTagVis:NiagaraInt32=&SpawnGroup()) {isSetParameters=true, module=SetVariables_#3}
   &ParticleState()
   &ScaleColor(
    Scale RGB=&VC_VectorBySpawnGroup(
     SpawnGroup or ExecIndex:NiagaraInt32=0
     CustomInt:NiagaraInt32=0
     MaximumSpawnGroups:NiagaraInt32=2
     Vector Group 0:Vector3f=&VectorFromCurve(VectorCurve=DataInterface() {propertyValues {
        XCurve {
         defaultValue=0
         keys [
          {tangentMode=RCTM_User, time=0.38, value=1}
          {arriveTangentWeight=0.02, tangentMode=RCTM_User, time=0.44, value=100}
          <at 0.5 1>
          {tangentMode=RCTM_Break, time=0.56, value=1}
          <at 0.57 10>
          {tangentMode=RCTM_User, time=1, value=1}
         ]
        }
        YCurve {
         defaultValue=0
         keys [
          {tangentMode=RCTM_User, time=0.38, value=1}
          {arriveTangentWeight=0.02, tangentMode=RCTM_User, time=0.44, value=100}
          <at 0.5 1>
          {tangentMode=RCTM_Break, time=0.56, value=1}
          <at 0.57 10>
          {tangentMode=RCTM_User, time=1, value=1}
         ]
        }
        ZCurve {keys [
          {tangentMode=RCTM_User, time=0.38, value=1}
          {arriveTangentWeight=0.02, tangentMode=RCTM_User, time=0.44, value=100}
          <at 0.5 1>
          {tangentMode=RCTM_Break, time=0.56, value=1}
          <at 0.57 10>
          {tangentMode=RCTM_User, time=1, value=1}
         ]}
       }})
     Vector Group 1:Vector3f=&VectorFromCurve(VectorCurve=DataInterface() {propertyValues {
        XCurve {
         defaultValue=0
         keys [
          {tangentMode=RCTM_User, time=0.44, value=1}
          {arriveTangentWeight=0.02, tangentMode=RCTM_User, time=0.5, value=10}
          {tangentMode=RCTM_User, time=1, value=1}
         ]
        }
        YCurve {
         defaultValue=0
         keys [
          {tangentMode=RCTM_User, time=0.44, value=1}
          {arriveTangentWeight=0.02, tangentMode=RCTM_User, time=0.5, value=10}
          {tangentMode=RCTM_User, time=1, value=1}
         ]
        }
        ZCurve {keys [
          {tangentMode=RCTM_User, time=0.44, value=1}
          {arriveTangentWeight=0.02, tangentMode=RCTM_User, time=0.5, value=10}
          {tangentMode=RCTM_User, time=1, value=1}
         ]}
       }})
     Vector Group 2:Vector3f=<0 0 0>
     Vector Group 3:Vector3f=<0 0 0>
     Vector Group 4:Vector3f=<0 0 0>
     Vector Group 5:Vector3f=<0 0 0>
    )
    ScaleA=true
    ScaleRGB=true
   )
   &UpdateMeshOrientation()
   &SolveForcesAndVelocity(Clamp Velocity=false, Limit Acceleration=false)
  ]
  rendererClasses [&NiagaraSpriteRendererProperties, &NiagaraMeshRendererProperties]
  renderers [
   {rendererClass=&NiagaraSpriteRendererProperties}
   {rendererClass=&NiagaraMeshRendererProperties, rendererIndex=1}
   {rendererClass=&NiagaraMeshRendererProperties, rendererIndex=2}
  ]
 }
}
summary {
 systemName=NS_Boost_Feline
 emitters [
  {
   emitterName=NE_VC_SingleBoost_TallFlares
   rendererClasses [&NiagaraSpriteRendererProperties, &NiagaraMeshRendererProperties]
  }
  {
   emitterName=NE_VC_SingleBoost_SmallFlares
   rendererClasses [&NiagaraSpriteRendererProperties, &NiagaraMeshRendererProperties]
  }
  {
   emitterName=NE_VC_SingleBoost_SmallFlares001
   rendererClasses [&NiagaraSpriteRendererProperties, &NiagaraMeshRendererProperties]
  }
  {
   emitterName=NE_VC_SingleBoost_TallFlares001
   rendererClasses [&NiagaraSpriteRendererProperties, &NiagaraMeshRendererProperties]
  }
 ]
 userVariables [
  User.bIsFrontendPreview {type=NiagaraBool}
  User.bIsPainted {type=NiagaraBool}
  User.ColorVariantID {type=NiagaraInt32}
  User.DTCarOffset {type=Vector3f, description="", defaultValue {struct=Vector3f, value=<0 0 0>}}
  User.GameModeID {type=NiagaraInt32}
  User.PaintedColor {type=LinearColor, description="", defaultValue {struct=LinearColor, value=<rgba 1 1 1 1>}}
  User.PaintedScalar {type=NiagaraFloat, defaultValue=1, description=""}
  User.PortCount {type=NiagaraInt32, defaultValue=1}
  User.SingleBoostPos {type=NiagaraPosition, description="", defaultValue {}}
 ]
}
